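; Looping gameplay cue for the Minor Spire's AOE charge beam effect, sizing a downward Niagara beam when the loop starts.
; GetAttackOrigin returns the instigator's location and a point a set distance along its negative up vector.
; A line trace between them, ignoring the instigator, gives a hit whose distance becomes the negative Z of the BeamEnd_0 parameter.
; The beam end is computed only once, in the loop-start event.
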(event EventOnLoopingStart(Niagara)
  (params (MyTarget :type Actor%20Object%20Reference) (Parameters :type Gameplay%20Cue%20Parameters%20Structure%20%28by%20ref%29) (ParticleComponents :type Array%20of%20Niagara%20Particle%20System%20Component%20Object%20References) (AudioComponents :type Array%20of%20Audio%20Component%20Object%20References))
  (bind (_normalizedmagnitude _rawmagnitude _effectcontext _matchedtagname _originaltag _aggregatedsourcetags _aggregatedtargettags _location _normal _instigator) (Ability|GameplayCue|BreakGameplayCueParameters Parameters))
  (bind (_start _end) (Default|GetAttackOrigin 2e+03 _instigator))
  (bind _output_get (Variables|Default|SetNiagara (Utilities|Array|Get(acopy) ParticleComponents)))
  (Transformation|AddWorldOffset _output_get "0, 0, 0" false false)
  (bind _outhit (Collision|LineTraceByChannel 0 _start _end "TraceTypeQuery10" false (Utilities|Array|MakeArray _instigator)))
  (Niagara|SetNiagaraVariable(Vector3) (Variables|Default|GetNiagara) "BeamEnd_0" (Math|Vector|MakeVector 0.0 0.0 (* (Math|Vector|Distance(Vector) (.Location (Collision|BreakHitResult _outhit)) (Transformation|GetActorLocation _instigator)) -1.0))))

(fn GetAttackOrigin
  (params (DistanceBelowOwner :type Float%20%28single-precision%29 :default 2e+03) (Target :type Actor%20Object%20Reference))
  (return (Transformation|GetActorLocation Target) (+ (* (Math|Vector|NegateVector (Transformation|GetActorUpVector Target)) DistanceBelowOwner) (Transformation|GetActorLocation Target))))

(fn ConstructionScript)
