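; Looping gameplay cue for the Slag shockwave charge effect, which adjusts its Niagara visuals when the target is a MaskMaker.
; On loop start it searches the target's class display name for "MaskMaker", then calls the parent handler.
; It passes the result to the first particle component as the IsMaskMakerSpireShards bool parameter.
; A leftover Print String of that flag has both screen and log output turned off.
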(event EventOnLoopingStart(Niagara)
  (params (MyTarget :type Actor%20Object%20Reference) (Parameters :type Gameplay%20Cue%20Parameters%20Structure%20%28by%20ref%29) (ParticleComponents :type Array%20of%20Niagara%20Particle%20System%20Component%20Object%20References) (AudioComponents :type Array%20of%20Audio%20Component%20Object%20References))
  (bind _returnvalue (Utilities|GetClass MyTarget))
  (bind _returnvalue_2 (Utilities|GetClassDisplayName _returnvalue))
  (bind _returnvalue_3 (Utilities|String|FindSubstring _returnvalue_2 "MaskMaker"))
  (bind _returnvalue_4 (== _returnvalue_3 -1))
  (bind _returnvalue_5 (not _returnvalue_4))
  (|CallParentFunction MyTarget Parameters ParticleComponents AudioComponents)
  (Development|PrintString 0 (Utilities|String|ToString(Boolean) _returnvalue_5) false false "(R=0.000000,G=0.660000,B=1.000000,A=1.000000)" 22.0)
  (Effects|Components|ParticleSystem|SetBoolParameter (Utilities|Array|Get(acopy) ParticleComponents) "IsMaskMakerSpireShards" _returnvalue_5))

(fn ConstructionScript)
