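; Burst gameplay cue for the Spire's player-targeted effect, configuring its spawned Niagara VFX.
; OnBurst takes the first particle component and sets its Duration from the cue's raw magnitude.
; The cue's SourceObject, cast to a skeletal mesh, feeds both mesh parameters, and its owner's location becomes Position_Spire.
; The cue's Normal is repurposed as a flag: an X of 1 sets IsMinor.
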(fn OnBurst(EitherVFX-preferthisversion)
  (params (MyTarget :type Actor%20Object%20Reference) (Parameters :type Gameplay%20Cue%20Parameters%20Structure%20%28by%20ref%29) (ParticleComponents :type Array%20of%20FXSystem%20Component%20Object%20References) (AudioComponents :type Array%20of%20Audio%20Component%20Object%20References) (BurstCameraShakeInstance :type Camera%20Shake%20Base%20Object%20Reference) (BurstDecalInstance :type Decal%20Actor%20Object%20Reference))
  (bind (_normalizedmagnitude _rawmagnitude _effectcontext _matchedtagname _originaltag _aggregatedsourcetags _aggregatedtargettags _location _normal _instigator _effectcauser _sourceobject) (Ability|GameplayCue|BreakGameplayCueParameters Parameters))
  (bind _asskeletal_mesh_component (Utilities|Casting|CastToSkeletalMeshComponent _sourceobject))
  (bind _output (Utilities|Array|Get(acopy) ParticleComponents))
  (bind _asniagara_particle_system_component (Utilities|Casting|CastToNiagaraComponent _output))
  (Effects|Components|ParticleSystem|SetFloatParameter _output "Duration" _rawmagnitude)
  (Niagara|SetNiagaraSkeletalMeshComponent _asniagara_particle_system_component "SKMesh" _asskeletal_mesh_component)
  (Niagara|SetNiagaraSkeletalMeshComponent _asniagara_particle_system_component "SKMesh_Minor" _asskeletal_mesh_component)
  (Niagara|SetNiagaraVariable(Bool) _asniagara_particle_system_component "IsMinor" (== (.x _normal) 1.0))
  (Niagara|SetNiagaraVariable(Position) _asniagara_particle_system_component "Position_Spire" (Transformation|GetActorLocation (Components|GetOwner _asskeletal_mesh_component)))
  (return))
